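# Physics asset for the ScorpionZero medium male player body cosmetic, covering its dangling parts: back skirts, front-right skirt, belt, fanny pack and hood.
# Every body is a single capsule, including pelvis, lower spine and thigh shapes alongside the dyn_ bones.
# Angular limits tie back skirts, hood and twist-only fanny pack to their MSTR bones, and chain the belt from MSTR through AIM to two segments.
# The front-right skirt has a body but no constraint.
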PhysicsAsset /BRCosmetics/Characters/Player/Male/Medium/Bodies/M_MED_ScorpionZero/Meshes/M_MED_ScorpionZero_Physics
bodies {
 dyn_back_skirt_l [{shapeType=Capsule, length=9, radius=3, shapeName=dyn_back_skirt_l_capsule, center=<6.9242 -2.6831 0>}]
 dyn_back_skirt_l_MSTR [{shapeType=Capsule, length=3, radius=2, shapeName=dyn_back_skirt_l_MSTR_capsule}]
 dyn_back_skirt_r [{shapeType=Capsule, length=9, radius=3, shapeName=dyn_back_skirt_r_capsule, center=<7.2241 2.5239 0>}]
 dyn_back_skirt_r_MSTR [{shapeType=Capsule, length=3, radius=2, shapeName=dyn_back_skirt_r_MSTR_capsule}]
 dyn_belt_mid_01 [{
   shapeType=Capsule
   length=5
   radius=2
   shapeName=dyn_belt_mid_01_capsule
   center=<3.5347 -0.204 0.3462>
   rotation=<rot 88.1421 -171.9664 -165.7522>
  }]
 dyn_belt_mid_02 [{shapeType=Capsule, length=5.5, radius=2.5, shapeName=dyn_belt_mid_02_capsule, center=<3.6035 -0.4023 0.45>, rotation=<rot 90 180 -175.7575>}]
 dyn_belt_mid_AIM [{shapeType=Capsule, length=3.5, radius=2, shapeName=dyn_belt_mid_AIM_capsule, center=<2.5582 0 1.03>, rotation=<rot 90 0 0>}]
 dyn_belt_mid_MSTR [{shapeType=Capsule, length=6, radius=1, shapeName=dyn_belt_mid_MSTR_capsule, rotation=<rot 90 90 0>}]
 dyn_fanny_pack [{shapeType=Capsule, length=8, radius=4, shapeName=dyn_fanny_pack_capsule, center=<2.6724 10.4897 -4.4198>, rotation=<rot 0 14 88>}]
 dyn_fanny_pack_MSTR [{shapeType=Capsule, length=3, radius=2, shapeName=dyn_fanny_pack_MSTR_capsule, rotation=<rot 0 0 90>}]
 dyn_front_skirt_r [{
   shapeType=Capsule
   length=15
   radius=5
   shapeName=dyn_front_skirt_r_capsule
   center=<14.6565 2.1192 -4.4152>
   rotation=<rot -0.0001 -0.0006 10.8147>
  }]
 dyn_hoody [{shapeType=Capsule, length=9.5, radius=3.5, shapeName=dyn_hoody_capsule, center=<10.6835 0 -0.1839>, rotation=<rot 0 0 90>}]
 dyn_hoody_MSTR [{shapeType=Capsule, length=3, radius=2, shapeName=dyn_hoody_MSTR_capsule, rotation=<rot 0 0 90>}]
 pelvis [{shapeType=Capsule, length=10, radius=13, shapeName=pelvis_capsule}]
 spine_01 [{shapeType=Capsule, length=7.5, radius=12, shapeName=spine_01_capsule, center=<5.6443 0 0>}]
 thigh_l [{shapeType=Capsule, length=15.5, radius=8.5, shapeName=thigh_l_capsule, center=<-8.8126 -1.6942 0>, rotation=<rot 90 0 0>}]
 thigh_r [{shapeType=Capsule, length=21, radius=9.5, shapeName=thigh_r_capsule, center=<13.2221 -0.4635 0>, rotation=<rot 90 180 -176.327>}]
}
constraints [
 {bone1Name=dyn_belt_mid_AIM, bone2Name=dyn_belt_mid_MSTR, swing1LimitDegrees=40, swing2LimitDegrees=20, twistLimitDegrees=0}
 {bone1Name=dyn_belt_mid_01, bone2Name=dyn_belt_mid_AIM, swing1LimitDegrees=45, swing2LimitDegrees=45, twistLimitDegrees=1}
 {bone1Name=dyn_belt_mid_02, bone2Name=dyn_belt_mid_01, swing1LimitDegrees=45, swing2LimitDegrees=45, twistLimitDegrees=1}
 {bone1Name=dyn_fanny_pack, bone2Name=dyn_fanny_pack_MSTR, swing1LimitDegrees=0, swing2LimitDegrees=0, twistLimitDegrees=20}
 {bone1Name=dyn_back_skirt_l, bone2Name=dyn_back_skirt_l_MSTR, swing1LimitDegrees=70, swing2LimitDegrees=10, twistLimitDegrees=5}
 {bone1Name=dyn_back_skirt_r, bone2Name=dyn_back_skirt_r_MSTR, swing1LimitDegrees=70, swing2LimitDegrees=10, twistLimitDegrees=5}
 {bone1Name=dyn_hoody, bone2Name=dyn_hoody_MSTR, swing1LimitDegrees=25, swing2LimitDegrees=25, twistLimitDegrees=15}
]
