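# Character-part AnimBlueprint for the ScorpionZero cosmetic body, adding secondary motion to hood, zipper, belts, sleeves and skirt.
# Copies the parent mesh pose through the Boredom deform Control Rig, then chains constraints, Anim Dynamics, RigidBody and skirt correctives driven by thigh angle and tactical sprint.
# Stationary crouch offsets thigh belts; bRigidBodyEnabled toggles final hood/mid-belt dynamics.
# EventGraph updates corrective alphas, gravity override, WPO material parameters and skydive-float wind.
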AnimBlueprint /BRCosmetics/Characters/Player/Male/Medium/Bodies/M_MED_ScorpionZero/Meshes/M_MED_ScorpionZero_AnimBP
target_skeleton=&/M_MED_ScorpionZero_Skeleton
graphs [
 AnimGraph {nodes [
   n1 = Anim Dynamics - Bone: dyn_hoody(ComponentPose=Local To Component(LocalPose=Use cached pose 'RB_Pose'()), Alpha=1, ExternalForce="0,0,0") {
    angularDampingOverride=0.999
    angularSpringConstant=500
    bOverrideAngularDamping=true
    bUseGravityOverride=true
    gravityScale=-1
    simulationSpace=RootRelative
    alphaBoolBlend {blendOption=Linear}
    boundBone {boneName=dyn_hoody}
    componentAppliedLinearAccClamp=<3000 3000 3000>
    componentLinearAccScale=<0.3 0.3 0.3>
    componentLinearVelScale=<0.3 0.3 0.3>
    gravityOverride {z=-900}
    physicsBodyDefinitions [{
      boundBone {boneName=dyn_hoody}
      boxExtents=<15 15 10>
      constraintSetup {angularLimitsMax=<0 0 15>, angularLimitsMin=<0 -40 -15>, angularTarget=<0 0 0>, linearAxesMax=<1 0 0>}
      localJointOffset=<5 0 0>
     }]
   }
   n2 = Anim Dynamics - Bone: dyn_belt_mid_AIM(ComponentPose=@n1, Alpha=1, ExternalForce="0,0,0") {
    angularDampingOverride=0.999
    angularSpringConstant=500
    bAngularSpring=true
    bOverrideAngularDamping=true
    bUseGravityOverride=true
    bUseSphericalLimits=true
    gravityScale=-1
    simulationSpace=RootRelative
    alphaBoolBlend {blendOption=Linear}
    boundBone {boneName=dyn_belt_mid_AIM}
    componentAppliedLinearAccClamp=<3000 3000 3000>
    componentLinearAccScale=<0.3 0.3 0.3>
    componentLinearVelScale=<0.3 0.3 0.3>
    gravityOverride {z=-980}
    physicsBodyDefinitions [{
      boundBone {boneName=dyn_belt_mid_AIM}
      boxExtents=<12 16 8>
      constraintSetup {angularLimitsMax=<0 0 30>, angularLimitsMin=<0 -60 -30>, angularTarget=<0 0 0>}
      localJointOffset=<3 0 0>
     }]
    sphericalLimits [{limitRadius=15, drivingBone {boneName=thigh_r}, sphereLocalOffset=<0 0 0>}]
   }
   n3 = Anim Dynamics - Bone: dyn_belt_mid_01(ComponentPose=@n2, Alpha=1, ExternalForce="0,0,0") {
    angularDampingOverride=0.999
    angularSpringConstant=500
    bAngularSpring=true
    bOverrideAngularDamping=true
    bUseGravityOverride=true
    bUseSphericalLimits=true
    gravityScale=-1
    simulationSpace=RootRelative
    alphaBoolBlend {blendOption=Linear}
    boundBone {boneName=dyn_belt_mid_01}
    componentAppliedLinearAccClamp=<3000 3000 3000>
    componentLinearAccScale=<0.3 0.3 0.3>
    componentLinearVelScale=<0.3 0.3 0.3>
    gravityOverride {z=-980}
    physicsBodyDefinitions [{
      sphereCollisionRadius=4
      boundBone {boneName=dyn_belt_mid_01}
      boxExtents=<16 10 10>
      constraintSetup {angularLimitsMax=<0 0 40>, angularLimitsMin=<0 -80 -40>, angularTarget=<0 0 0>}
      localJointOffset=<3 0 0>
     }]
    sphericalLimits [{limitRadius=15, drivingBone {boneName=thigh_r}, sphereLocalOffset=<5 0 0>}]
   }
   n4 = Anim Dynamics - Bone: dyn_belt_mid_02(ComponentPose=@n3, Alpha=1, ExternalForce="0,0,0") {
    angularDampingOverride=0.999
    angularSpringConstant=500
    bAngularSpring=true
    bOverrideAngularDamping=true
    bUseGravityOverride=true
    bUseSphericalLimits=true
    gravityScale=-1
    simulationSpace=RootRelative
    alphaBoolBlend {blendOption=Linear}
    boundBone {boneName=dyn_belt_mid_02}
    componentAppliedLinearAccClamp=<3000 3000 3000>
    componentLinearAccScale=<0.3 0.3 0.3>
    componentLinearVelScale=<0.3 0.3 0.3>
    gravityOverride {z=-980}
    physicsBodyDefinitions [{
      sphereCollisionRadius=2
      boundBone {boneName=dyn_belt_mid_02}
      boxExtents=<16 10 10>
      constraintSetup {angularLimitsMax=<0 0 40>, angularLimitsMin=<0 -80 -40>, angularTarget=<0 0 0>}
      localJointOffset=<3 0 0>
     }]
    sphericalLimits [
     {limitRadius=12, drivingBone {boneName=thigh_r}, sphereLocalOffset=<12 0 0>}
     {limitRadius=12, drivingBone {boneName=thigh_r}, sphereLocalOffset=<8 0 0>}
    ]
   }
   Output Pose(Result=Blend Poses by bool(BlendPose_0=Use cached pose 'RB_Pose'(), BlendPose_1=Component To Local(ComponentPose=@n4), bActiveValue=Get bRigidBodyEnabled()) {blendType=Linear}) {layerGroup=None}
   n5 = "Transform (Modify) Bone - Bone: dyn_thigh_l_belt_01_MSTR"(ComponentPose=Local To Component(LocalPose=Use cached pose 'Base_Pose'()), Alpha=1, Translation="0.5,0,5", Rotation="0,0,0", Scale="1,1,1") {
    rotationMode=BMM_Additive
    rotationSpace=BCS_BoneSpace
    translationMode=BMM_Additive
    translationSpace=BCS_BoneSpace
    alphaBoolBlend {blendOption=Linear}
    boneToModify {boneName=dyn_thigh_l_belt_01_MSTR}
    translation {x=0.5, z=5}
   }
   n6 = "Transform (Modify) Bone - Bone: dyn_thigh_l_belt_03_MSTR"(ComponentPose=@n5, Alpha=1, Translation="0.5,0,5", Rotation="0,0,0", Scale="1,1,1") {
    rotationMode=BMM_Additive
    rotationSpace=BCS_BoneSpace
    translationMode=BMM_Additive
    translationSpace=BCS_BoneSpace
    alphaBoolBlend {blendOption=Linear}
    boneToModify {boneName=dyn_thigh_l_belt_03_MSTR}
    translation {x=0.5, z=5}
   }
   n7 = "Transform (Modify) Bone - Bone: dyn_thigh_l_belt_02_MSTR"(ComponentPose=@n6, Alpha=1, Translation="0.5,0,1.5", Rotation="0,0,0", Scale="1,1,1") {
    rotationMode=BMM_Additive
    rotationSpace=BCS_BoneSpace
    translationMode=BMM_Additive
    translationSpace=BCS_BoneSpace
    alphaBoolBlend {blendOption=Linear}
    boneToModify {boneName=dyn_thigh_l_belt_02_MSTR}
    translation {x=0.5, z=1.5}
   }
   n8 = "Transform (Modify) Bone - Bone: dyn_thigh_l_belt_04_MSTR"(ComponentPose=@n7, Alpha=1, Translation="0.5,0,1.5", Rotation="0,0,0", Scale="1,1,1") {
    rotationMode=BMM_Additive
    rotationSpace=BCS_BoneSpace
    translationMode=BMM_Additive
    translationSpace=BCS_BoneSpace
    alphaBoolBlend {blendOption=Linear}
    boneToModify {boneName=dyn_thigh_l_belt_04_MSTR}
    translation {x=0.5, z=1.5}
   }
   n9 = "Transform (Modify) Bone - Bone: dyn_thigh_r_belt_04_MSTR"(ComponentPose=@n8, Alpha=1, Translation="0.5,0,7", Rotation="0,0,0", Scale="1,1,1") {
    rotationMode=BMM_Additive
    rotationSpace=BCS_BoneSpace
    translationMode=BMM_Additive
    translationSpace=BCS_BoneSpace
    alphaBoolBlend {blendOption=Linear}
    boneToModify {boneName=dyn_thigh_r_belt_04_MSTR}
    translation {x=0.5, z=7}
   }
   n10 = "Transform (Modify) Bone - Bone: dyn_thigh_r_belt_02_MSTR"(ComponentPose=@n9, Alpha=1, Translation="0.5,0,7", Rotation="0,0,0", Scale="1,1,1") {
    rotationMode=BMM_Additive
    rotationSpace=BCS_BoneSpace
    translationMode=BMM_Additive
    translationSpace=BCS_BoneSpace
    alphaBoolBlend {blendOption=Linear}
    boneToModify {boneName=dyn_thigh_r_belt_02_MSTR}
    translation {x=0.5, z=7}
   }
   n11 = "Transform (Modify) Bone - Bone: dyn_thigh_r_belt_03_MSTR"(ComponentPose=@n10, Alpha=1, Translation="0.5,0,6", Rotation="0,0,0", Scale="1,1,1") {
    rotationMode=BMM_Additive
    rotationSpace=BCS_BoneSpace
    translationMode=BMM_Additive
    translationSpace=BCS_BoneSpace
    alphaBoolBlend {blendOption=Linear}
    boneToModify {boneName=dyn_thigh_r_belt_03_MSTR}
    translation {x=0.5, z=6}
   }
   n12 = "Transform (Modify) Bone - Bone: dyn_thigh_r_belt_01_MSTR"(ComponentPose=@n11, Alpha=1, Translation="0.5,0,6", Rotation="0,0,0", Scale="1,1,1") {
    rotationMode=BMM_Additive
    rotationSpace=BCS_BoneSpace
    translationMode=BMM_Additive
    translationSpace=BCS_BoneSpace
    alphaBoolBlend {blendOption=Linear}
    boneToModify {boneName=dyn_thigh_r_belt_01_MSTR}
    translation {x=0.5, z=6}
   }
   Save cached pose 'RB_Pose'(Pose=Blend Poses by bool(
     BlendPose_0=Component To Local(ComponentPose=@n12)
     BlendPose_1=Use cached pose 'Base_Pose'()
     BlendTime_1=0.4
     bActiveValue=Get Crouch stop() -> (Crouch stop:bool)
    ) {blendType=Linear}) {cacheName=RB_Pose}
   n13 = Control Rig(Source=Copy Pose From Mesh() {bUseAttachedParent=true}, Alpha=0.8) {
    alphaBoolBlend {blendOption=Linear}
    controlRigAssetReference {blueprintRigClass=/Game/Characters/Player/Male/Medium/Bodies/M_MED_Boredom/Meshes/M_MED_DeformRig_Boredom_ControlRig.M_MED_DeformRig_Boredom_ControlRig_C}
   }
   n14 = "Constraint - dyn_zipper_mid_MSTR (PPP)"!(
    ComponentPose=Local To Component(LocalPose=@n13)
    Alpha=1
    ConstraintWeights_0=0
    ConstraintWeights_1:float=0.555
    ConstraintWeights_2:float=0.445
   ) -> (Pose) {
    alphaBoolBlend {blendOption=Linear}
    boneToModify {boneName=dyn_zipper_mid_MSTR}
    constraintSetup [{targetBone {boneName=head}}, {targetBone {boneName=neck_02}}, {targetBone {boneName=spine_05}}]
    constraintWeights [1,1,1]
   }
   n15 = "Constraint - dyn_belt_mid_MSTR (PPPP)"!(
    ComponentPose=@n14
    Alpha=1
    ConstraintWeights_0=0.3
    ConstraintWeights_1:float=0.24
    ConstraintWeights_2:float=0.33
    ConstraintWeights_3:float=0.13
   ) -> (Pose) {
    alphaBoolBlend {blendOption=Linear}
    boneToModify {boneName=dyn_belt_mid_MSTR}
    constraintSetup [{targetBone {boneName=pelvis}}, {targetBone {boneName=spine_01}}, {targetBone {boneName=spine_02}}, {targetBone {boneName=spine_03}}]
    constraintWeights [1,1,1,1]
   }
   n16 = "Constraint - dyn_hoody_MSTR (PP)"!(ComponentPose=@n15, Alpha=1, ConstraintWeights_0=0.8, ConstraintWeights_1:float=0.2) -> (Pose) {
    alphaBoolBlend {blendOption=Linear}
    boneToModify {boneName=dyn_hoody_MSTR}
    constraintSetup [{targetBone {boneName=spine_05}}, {targetBone {boneName=neck_02}}]
    constraintWeights [1,1]
   }
   n17 = "Constraint - dyn_thigh_l_belt_02_MSTR (PPP)"!(ComponentPose=@n16, Alpha=1, ConstraintWeights_0=0, ConstraintWeights_1:float=0.5, ConstraintWeights_2:float=0.5) -> (Pose) {
    alphaBoolBlend {blendOption=Linear}
    boneToModify {boneName=dyn_thigh_l_belt_02_MSTR}
    constraintSetup [{targetBone {boneName=pelvis}}, {targetBone {boneName=thigh_l}}, {targetBone {boneName=thigh_twist_01_l}}]
    constraintWeights [1,1,1]
   }
   n18 = "Constraint - dyn_thigh_l_belt_01_MSTR (PPP)"!(ComponentPose=@n17, Alpha=1, ConstraintWeights_0=0, ConstraintWeights_1:float=0.85, ConstraintWeights_2:float=0.15) -> (Pose) {
    alphaBoolBlend {blendOption=Linear}
    boneToModify {boneName=dyn_thigh_l_belt_01_MSTR}
    constraintSetup [{targetBone {boneName=pelvis}}, {targetBone {boneName=thigh_l}}, {targetBone {boneName=thigh_twist_01_l}}]
    constraintWeights [1,1,1]
   }
   n19 = "Constraint - dyn_thigh_r_belt_01_MSTR (PPP)"!(ComponentPose=@n18, Alpha=1, ConstraintWeights_0=0, ConstraintWeights_1:float=0.5, ConstraintWeights_2:float=0.5) -> (Pose) {
    alphaBoolBlend {blendOption=Linear}
    boneToModify {boneName=dyn_thigh_r_belt_01_MSTR}
    constraintSetup [{targetBone {boneName=pelvis}}, {targetBone {boneName=thigh_r}}, {targetBone {boneName=thigh_twist_01_r}}]
    constraintWeights [1,1,1]
   }
   n20 = "Constraint - dyn_thigh_r_belt_02_MSTR (PPP)"!(ComponentPose=@n19, Alpha=1, ConstraintWeights_0=0, ConstraintWeights_1:float=0.85, ConstraintWeights_2:float=0.15) -> (Pose) {
    alphaBoolBlend {blendOption=Linear}
    boneToModify {boneName=dyn_thigh_r_belt_02_MSTR}
    constraintSetup [{targetBone {boneName=pelvis}}, {targetBone {boneName=thigh_r}}, {targetBone {boneName=thigh_twist_01_r}}]
    constraintWeights [1,1,1]
   }
   n21 = "Constraint - dyn_thigh_l_belt_04_MSTR (PPP)"!(ComponentPose=@n20, Alpha=1, ConstraintWeights_0=0, ConstraintWeights_1:float=0.5, ConstraintWeights_2:float=0.5) -> (Pose) {
    alphaBoolBlend {blendOption=Linear}
    boneToModify {boneName=dyn_thigh_l_belt_04_MSTR}
    constraintSetup [{targetBone {boneName=pelvis}}, {targetBone {boneName=thigh_l}}, {targetBone {boneName=thigh_twist_01_l}}]
    constraintWeights [1,1,1]
   }
   n22 = "Constraint - dyn_thigh_l_belt_03_MSTR (PPP)"!(ComponentPose=@n21, Alpha=1, ConstraintWeights_0=0, ConstraintWeights_1:float=0.85, ConstraintWeights_2:float=0.15) -> (Pose) {
    alphaBoolBlend {blendOption=Linear}
    boneToModify {boneName=dyn_thigh_l_belt_03_MSTR}
    constraintSetup [{targetBone {boneName=pelvis}}, {targetBone {boneName=thigh_l}}, {targetBone {boneName=thigh_twist_01_l}}]
    constraintWeights [1,1,1]
   }
   n23 = "Constraint - dyn_thigh_r_belt_03_MSTR (PPP)"!(ComponentPose=@n22, Alpha=1, ConstraintWeights_0=0, ConstraintWeights_1:float=0.5, ConstraintWeights_2:float=0.5) -> (Pose) {
    alphaBoolBlend {blendOption=Linear}
    boneToModify {boneName=dyn_thigh_r_belt_03_MSTR}
    constraintSetup [{targetBone {boneName=pelvis}}, {targetBone {boneName=thigh_r}}, {targetBone {boneName=thigh_twist_01_r}}]
    constraintWeights [1,1,1]
   }
   n24 = "Constraint - dyn_thigh_r_belt_04_MSTR (PPP)"!(ComponentPose=@n23, Alpha=1, ConstraintWeights_0=0, ConstraintWeights_1:float=0.85, ConstraintWeights_2:float=0.15) -> (Pose) {
    alphaBoolBlend {blendOption=Linear}
    boneToModify {boneName=dyn_thigh_r_belt_04_MSTR}
    constraintSetup [{targetBone {boneName=pelvis}}, {targetBone {boneName=thigh_r}}, {targetBone {boneName=thigh_twist_01_r}}]
    constraintWeights [1,1,1]
   }
   n25 = Anim Dynamics - Bone: dyn_zipper_mid_01(ComponentPose=@n24, Alpha=1, ExternalForce="0,0,0") {
    angularDampingOverride=0.99
    angularSpringConstant=500
    bOverrideAngularDamping=true
    bUseGravityOverride=true
    bUseSphericalLimits=true
    gravityScale=-1
    simulationSpace=RootRelative
    alphaBoolBlend {blendOption=Linear}
    boundBone {boneName=dyn_zipper_mid_01}
    componentAppliedLinearAccClamp=<5000 5000 5000>
    componentLinearAccScale=<0.2 0.2 0.2>
    componentLinearVelScale=<0.2 0.2 0.2>
    gravityOverride {z=-980}
    physicsBodyDefinitions [{
      collisionType=CustomSphere
      sphereCollisionRadius=1
      boundBone {boneName=dyn_zipper_mid_01}
      boxExtents=<10 5 2.5>
      constraintSetup {angularLimitsMax=<0 15 30>, angularLimitsMin=<0 -100 -30>, angularTarget=<0 0 0>}
      localJointOffset=<3 0 0>
     }]
    sphericalLimits [
     {limitRadius=3, drivingBone {boneName=spine_05}, sphereLocalOffset=<6.5 7.5 0>}
     {limitRadius=3, drivingBone {boneName=spine_05}, sphereLocalOffset=<6.5 7 -2>}
    ]
   }
   n26 = Anim Dynamics - Bone: dyn_roop_r(ComponentPose=@n25, Alpha=1, ExternalForce="0,0,0") {
    angularDampingOverride=0.9999
    angularSpringConstant=500
    bOverrideAngularDamping=true
    bUseGravityOverride=true
    gravityScale=-1
    simulationSpace=RootRelative
    alphaBoolBlend {blendOption=Linear}
    boundBone {boneName=dyn_roop_r}
    componentAppliedLinearAccClamp=<3000 3000 3000>
    componentLinearAccScale=<0.3 0.3 0.3>
    componentLinearVelScale=<0.3 0.3 0.3>
    gravityOverride {z=-800}
    physicsBodyDefinitions [{
      boundBone {boneName=dyn_roop_r}
      boxExtents=<8 8 5>
      constraintSetup {angularLimitsMax=<10 -10 30>, angularLimitsMin=<-10 -60 -30>, angularTarget=<0 0 0>, linearAxesMax=<1 0 0>}
      localJointOffset=<3 0 0>
     }]
   }
   n27 = Anim Dynamics - Bone: dyn_thigh_l_belt_01(ComponentPose=@n26, Alpha=1, ExternalForce="0,0,0") {
    angularDampingOverride=0.8
    angularSpringConstant=500
    bAngularSpring=true
    bOverrideAngularDamping=true
    bUseGravityOverride=true
    gravityScale=-1
    simulationSpace=RootRelative
    alphaBoolBlend {blendOption=Linear}
    boundBone {boneName=dyn_thigh_l_belt_01}
    componentAppliedLinearAccClamp=<3000 3000 3000>
    componentLinearAccScale=<0.3 0.3 0.3>
    componentLinearVelScale=<0.3 0.3 0.3>
    gravityOverride {z=-980}
    physicsBodyDefinitions [{
      boundBone {boneName=dyn_thigh_l_belt_01}
      boxExtents=<8 10 3>
      constraintSetup {angularLimitsMax=<5 10 5>, angularLimitsMin=<-5 -50 -5>, angularTarget=<0 0 0>, linearAxesMax=<1 0 0>}
      localJointOffset=<8 0 0>
     }]
   }
   n28 = Anim Dynamics - Bone: dyn_thigh_l_belt_03(ComponentPose=@n27, Alpha=1, ExternalForce="0,0,0") {
    angularDampingOverride=0.8
    angularSpringConstant=500
    bAngularSpring=true
    bOverrideAngularDamping=true
    bUseGravityOverride=true
    gravityScale=-1
    simulationSpace=RootRelative
    alphaBoolBlend {blendOption=Linear}
    boundBone {boneName=dyn_thigh_l_belt_03}
    componentAppliedLinearAccClamp=<3000 3000 3000>
    componentLinearAccScale=<0.3 0.3 0.3>
    componentLinearVelScale=<0.3 0.3 0.3>
    gravityOverride {z=-980}
    physicsBodyDefinitions [{
      boundBone {boneName=dyn_thigh_l_belt_03}
      boxExtents=<10 10 3>
      constraintSetup {angularLimitsMax=<5 0 5>, angularLimitsMin=<-5 -50 -5>, angularTarget=<0 0 0>, linearAxesMax=<1 0 0>}
      localJointOffset=<6 0 0>
     }]
   }
   n29 = Anim Dynamics - Bone: dyn_thigh_r_belt_01(ComponentPose=@n28, Alpha=1, ExternalForce="0,0,0") {
    angularDampingOverride=0.99
    angularSpringConstant=500
    bAngularSpring=true
    bOverrideAngularDamping=true
    bUseGravityOverride=true
    gravityScale=-1
    simulationSpace=RootRelative
    alphaBoolBlend {blendOption=Linear}
    boundBone {boneName=dyn_thigh_r_belt_01}
    componentAppliedLinearAccClamp=<3000 3000 3000>
    componentLinearAccScale=<0.3 0.3 0.3>
    componentLinearVelScale=<0.3 0.3 0.3>
    gravityOverride {z=-980}
    physicsBodyDefinitions [{
      boundBone {boneName=dyn_thigh_r_belt_01}
      boxExtents=<10 10 3>
      constraintSetup {angularLimitsMax=<5 15 5>, angularLimitsMin=<-5 -50 -5>, angularTarget=<0 0 0>, linearAxesMax=<1 0 0>}
      localJointOffset=<1 0 0>
     }]
   }
   n30 = Anim Dynamics - Bone: dyn_thigh_r_belt_03(ComponentPose=@n29, Alpha=1, ExternalForce="0,0,0") {
    angularDampingOverride=0.99
    angularSpringConstant=500
    bAngularSpring=true
    bOverrideAngularDamping=true
    bUseGravityOverride=true
    gravityScale=-1
    simulationSpace=RootRelative
    alphaBoolBlend {blendOption=Linear}
    boundBone {boneName=dyn_thigh_r_belt_03}
    componentAppliedLinearAccClamp=<3000 3000 3000>
    componentLinearAccScale=<0.3 0.3 0.3>
    componentLinearVelScale=<0.3 0.3 0.3>
    gravityOverride {z=-980}
    physicsBodyDefinitions [{
      boundBone {boneName=dyn_thigh_r_belt_03}
      boxExtents=<10 10 3>
      constraintSetup {angularLimitsMax=<5 10 5>, angularLimitsMin=<-5 -50 -5>, angularTarget=<0 0 0>, linearAxesMax=<1 0 0>}
      localJointOffset=<5 0 0>
     }]
   }
   n31 = Anim Dynamics - Bone: dyn_sleave_dn_r(ComponentPose=@n30, Alpha=1, ExternalForce="0,0,0") {
    angularDampingOverride=0.8
    angularSpringConstant=500
    bAngularSpring=true
    bOverrideAngularDamping=true
    bUseGravityOverride=true
    gravityScale=-1
    simulationSpace=RootRelative
    alphaBoolBlend {blendOption=Linear}
    boundBone {boneName=dyn_sleave_dn_r}
    componentAppliedLinearAccClamp=<3000 3000 3000>
    componentLinearAccScale=<0.3 0.3 0.3>
    componentLinearVelScale=<0.3 0.3 0.3>
    gravityOverride {z=-980}
    physicsBodyDefinitions [{
      boundBone {boneName=dyn_sleave_dn_r}
      boxExtents=<10 10 3>
      constraintSetup {angularLimitsMax=<5 5 2>, angularLimitsMin=<-5 -20 -2>, angularTarget=<0 0 0>}
      localJointOffset=<-5 0 0>
     }]
    relativeSpaceBone {boneName=spine_05}
   }
   n32 = Anim Dynamics - Bone: dyn_sleave_dn_l(ComponentPose=@n31, Alpha=1, ExternalForce="0,0,0") {
    angularDampingOverride=0.8
    angularSpringConstant=500
    bAngularSpring=true
    bOverrideAngularDamping=true
    bUseGravityOverride=true
    gravityScale=-1
    simulationSpace=RootRelative
    alphaBoolBlend {blendOption=Linear}
    boundBone {boneName=dyn_sleave_dn_l}
    componentAppliedLinearAccClamp=<3000 3000 3000>
    componentLinearAccScale=<0.3 0.3 0.3>
    componentLinearVelScale=<0.3 0.3 0.3>
    gravityOverride {z=-980}
    physicsBodyDefinitions [{
      boundBone {boneName=dyn_sleave_dn_l}
      boxExtents=<5 10 3>
      constraintSetup {angularLimitsMax=<5 5 2>, angularLimitsMin=<-5 -20 -2>, angularTarget=<0 0 0>}
      localJointOffset=<5 0 0>
     }]
    relativeSpaceBone {boneName=spine_05}
   }
   n33 = Anim Dynamics - Bone: dyn_loop_l(ComponentPose=@n32, Alpha=1, ExternalForce="0,0,0") {
    angularDampingOverride=0.9999
    angularSpringConstant=500
    bOverrideAngularDamping=true
    bUseGravityOverride=true
    gravityScale=-1
    simulationSpace=RootRelative
    alphaBoolBlend {blendOption=Linear}
    boundBone {boneName=dyn_loop_l}
    componentAppliedLinearAccClamp=<3000 3000 3000>
    componentLinearAccScale=<0.3 0.3 0.3>
    componentLinearVelScale=<0.3 0.3 0.3>
    gravityOverride {z=-800}
    physicsBodyDefinitions [{
      boundBone {boneName=dyn_loop_l}
      boxExtents=<7 7 4>
      constraintSetup {angularLimitsMax=<10 0 30>, angularLimitsMin=<-10 -60 -30>, angularTarget=<0 0 0>, linearAxesMax=<1 0 0>}
      localJointOffset=<3 0 0>
     }]
   }
   n34 = Anim Dynamics - Bone: dyn_thigh_l_belt_02(ComponentPose=@n33, Alpha=1, ExternalForce="0,0,0") {
    angularDampingOverride=0.8
    angularSpringConstant=500
    bAngularSpring=true
    bOverrideAngularDamping=true
    bUseGravityOverride=true
    gravityScale=-1
    simulationSpace=RootRelative
    alphaBoolBlend {blendOption=Linear}
    boundBone {boneName=dyn_thigh_l_belt_02}
    componentAppliedLinearAccClamp=<3000 3000 3000>
    componentLinearAccScale=<0.3 0.3 0.3>
    componentLinearVelScale=<0.3 0.3 0.3>
    gravityOverride {z=-980}
    physicsBodyDefinitions [{
      boundBone {boneName=dyn_thigh_l_belt_02}
      boxExtents=<5 10 3>
      constraintSetup {angularLimitsMax=<5 10 5>, angularLimitsMin=<-5 -50 -5>, angularTarget=<0 0 0>, linearAxesMax=<1 0 0>}
      localJointOffset=<6 0 0>
     }]
   }
   n35 = Anim Dynamics - Bone: dyn_thigh_l_belt_04(ComponentPose=@n34, Alpha=1, ExternalForce="0,0,0") {
    angularDampingOverride=0.8
    angularSpringConstant=500
    bAngularSpring=true
    bOverrideAngularDamping=true
    bUseGravityOverride=true
    gravityScale=-1
    simulationSpace=RootRelative
    alphaBoolBlend {blendOption=Linear}
    boundBone {boneName=dyn_thigh_l_belt_04}
    componentAppliedLinearAccClamp=<3000 3000 3000>
    componentLinearAccScale=<0.3 0.3 0.3>
    componentLinearVelScale=<0.3 0.3 0.3>
    gravityOverride {z=-980}
    physicsBodyDefinitions [{
      boundBone {boneName=dyn_thigh_l_belt_04}
      boxExtents=<10 10 3>
      constraintSetup {angularLimitsMax=<5 10 5>, angularLimitsMin=<-5 -50 -5>, angularTarget=<0 0 0>, linearAxesMax=<1 0 0>}
      localJointOffset=<7 0 0>
     }]
   }
   n36 = Anim Dynamics - Bone: dyn_thigh_r_belt_02(ComponentPose=@n35, Alpha=1, ExternalForce="0,0,0") {
    angularDampingOverride=0.99
    angularSpringConstant=500
    bAngularSpring=true
    bOverrideAngularDamping=true
    bUseGravityOverride=true
    gravityScale=-1
    simulationSpace=RootRelative
    alphaBoolBlend {blendOption=Linear}
    boundBone {boneName=dyn_thigh_r_belt_02}
    componentAppliedLinearAccClamp=<3000 3000 3000>
    componentLinearAccScale=<0.3 0.3 0.3>
    componentLinearVelScale=<0.3 0.3 0.3>
    gravityOverride {z=-980}
    physicsBodyDefinitions [{
      boundBone {boneName=dyn_thigh_r_belt_02}
      boxExtents=<10 10 3>
      constraintSetup {angularLimitsMax=<5 0 5>, angularLimitsMin=<-5 -50 -5>, angularTarget=<0 0 0>, linearAxesMax=<1 0 0>}
      localJointOffset=<4 0 0>
     }]
   }
   n37 = Anim Dynamics - Bone: dyn_thigh_r_belt_04(ComponentPose=@n36, Alpha=1, ExternalForce="0,0,0") {
    angularDampingOverride=0.99
    angularSpringConstant=200
    bAngularSpring=true
    bOverrideAngularDamping=true
    bUseGravityOverride=true
    gravityScale=-1
    simulationSpace=RootRelative
    alphaBoolBlend {blendOption=Linear}
    boundBone {boneName=dyn_thigh_r_belt_04}
    componentAppliedLinearAccClamp=<3000 3000 3000>
    componentLinearAccScale=<0.3 0.3 0.3>
    componentLinearVelScale=<0.3 0.3 0.3>
    gravityOverride {z=-980}
    physicsBodyDefinitions [{
      boundBone {boneName=dyn_thigh_r_belt_04}
      boxExtents=<10 10 3>
      constraintSetup {angularLimitsMax=<5 0 5>, angularLimitsMin=<-5 -50 -5>, angularTarget=<0 0 0>, linearAxesMax=<1 0 0>}
      localJointOffset=<1 0 0>
     }]
   }
   n38 = Anim Dynamics - Bone: dyn_sleave_up_r(ComponentPose=@n37, Alpha=1, ExternalForce="0,0,0") {
    angularDampingOverride=0.8
    angularSpringConstant=500
    bAngularSpring=true
    bOverrideAngularDamping=true
    bUseGravityOverride=true
    gravityScale=-1
    simulationSpace=RootRelative
    alphaBoolBlend {blendOption=Linear}
    boundBone {boneName=dyn_sleave_up_r}
    componentAppliedLinearAccClamp=<3000 3000 3000>
    componentLinearAccScale=<0.3 0.3 0.3>
    componentLinearVelScale=<0.3 0.3 0.3>
    gravityOverride {z=-980}
    physicsBodyDefinitions [{
      boundBone {boneName=dyn_sleave_up_r}
      boxExtents=<10 10 5>
      constraintSetup {angularLimitsMax=<5 10 2>, angularLimitsMin=<-5 5 -2>, angularTarget=<0 0 0>}
      localJointOffset=<-5 0 0>
     }]
    relativeSpaceBone {boneName=spine_05}
   }
   n39 = Anim Dynamics - Bone: dyn_sleave_up_l(ComponentPose=@n38, Alpha=1, ExternalForce="0,0,0") {
    angularDampingOverride=0.8
    angularSpringConstant=500
    bAngularSpring=true
    bOverrideAngularDamping=true
    bUseGravityOverride=true
    gravityScale=-1
    simulationSpace=RootRelative
    alphaBoolBlend {blendOption=Linear}
    boundBone {boneName=dyn_sleave_up_l}
    componentAppliedLinearAccClamp=<3000 3000 3000>
    componentLinearAccScale=<0.3 0.3 0.3>
    componentLinearVelScale=<0.3 0.3 0.3>
    gravityOverride {z=-980}
    physicsBodyDefinitions [{
      boundBone {boneName=dyn_sleave_up_l}
      boxExtents=<10 10 5>
      constraintSetup {angularLimitsMax=<5 10 2>, angularLimitsMin=<-5 5 -2>, angularTarget=<0 0 0>}
      localJointOffset=<5 0 0>
     }]
    relativeSpaceBone {boneName=spine_05}
   }
   n40 = RigidBody!(ComponentPose=@n39, Alpha, OverrideWorldGravity:Vector=Get Gravity Override(), ExternalForce) -> (Pose) {
    bClampLinearTranslationLimitToRefPose=true
    bOverrideWorldGravity=true
    overridePhysicsAsset=&/M_MED_ScorpionZero_Physics
    simulationSpace=BaseBoneSpace
    alphaBoolBlend {blendOption=Linear}
    baseBoneRef {boneName=spine_02}
    componentAppliedLinearAccClamp=<3000 3000 3000>
    componentLinearAccScale=<0.4 0.4 0.5>
    componentLinearVelScale=<0.4 0.4 0.5>
    overrideWorldGravity {z=-980}
   }
   n41 = "Transform (Modify) Bone - Bone: dyn_front_skirt_l_MSTR"(ComponentPose=@n40, Alpha=Get Alpha_ThighL_Front() -> (Alpha_ThighL_Front:double), Translation="-12,9,-2", Rotation="0,30,0", Scale="1,1,1") {
    rotationMode=BMM_Additive
    rotationSpace=BCS_BoneSpace
    translationMode=BMM_Additive
    translationSpace=BCS_BoneSpace
    alphaBoolBlend {blendOption=Linear}
    boneToModify {boneName=dyn_front_skirt_l_MSTR}
    rotation {yaw=30}
    translation=<-12 9 -2>
   }
   n42 = "Transform (Modify) Bone - Bone: dyn_front_skirt_r_MSTR"(ComponentPose=@n41, Alpha=Get Alpha_ThighR_Front() -> (Alpha_ThighR_Front:double), Translation="10,-3.3,2", Rotation="0,10,0", Scale="1,1,1") {
    rotationMode=BMM_Additive
    rotationSpace=BCS_BoneSpace
    translationMode=BMM_Additive
    translationSpace=BCS_BoneSpace
    alphaBoolBlend {blendOption=Linear}
    boneToModify {boneName=dyn_front_skirt_r_MSTR}
    rotation {yaw=10}
    translation=<10 -3.3 2>
   }
   n43 = Get bIsTacticalSprinting()
   n44 = "Transform (Modify) Bone - Bone: dyn_back_skirt_l_MSTR"!(ComponentPose=@n42, bAlphaBoolEnabled:bool=@n43, Translation="0,0,0", Rotation="0,-10,0", Scale="1,1,1") -> (Pose) {
    alphaInputType=Bool
    rotationMode=BMM_Additive
    rotationSpace=BCS_BoneSpace
    translationMode=BMM_Additive
    translationSpace=BCS_BoneSpace
    alphaBoolBlend {blendOption=Linear}
    boneToModify {boneName=dyn_back_skirt_l_MSTR}
    rotation {yaw=-10}
   }
   n45 = "Transform (Modify) Bone - Bone: dyn_back_skirt_r_MSTR"!(ComponentPose=@n44, bAlphaBoolEnabled:bool=@n43, Translation="0,0,0", Rotation="0,10,0", Scale="1,1,1") -> (Pose) {
    alphaInputType=Bool
    rotationMode=BMM_Additive
    rotationSpace=BCS_BoneSpace
    translationMode=BMM_Additive
    translationSpace=BCS_BoneSpace
    alphaBoolBlend {blendOption=Linear}
    boneToModify {boneName=dyn_back_skirt_r_MSTR}
    rotation {yaw=10}
   }
   Save cached pose 'Base_Pose'(Pose=Component To Local(ComponentPose=@n45)) {cacheName=Base_Pose}
  ]}
 EventGraph {nodes [
   Event BlueprintInitializeAnimation()
   > Set Skel Mesh(Skel Mesh=GetOwnerSkeletalMeshComponent())
   n1 = TryGetPawnOwner()
   > Is Valid(InputObject=@n1)
   > n2 = Cast To FortPlayerPawnAthena(Object=@n1)
   > Set PlayerPawn(PlayerPawn:Fort Player Pawn Athena Object Reference=@n2) -> (Output_Get:Fort Player Pawn Athena Object Reference)
   n3 = Event BlueprintUpdateAnimation()
   > n4 = Sequence!(execute) -> (then_0, then_1, then_2:Exec)
   n5 = Get Skel Mesh()
   Set Alpha_ThighL_Front(
    execute=@n4.then_1
    Alpha_ThighL_Front:double=GetAlphaFromReferenceTransformDelta(SkeletalMeshComponent=@n5, BoneName=thigh_l, BaseName=pelvis, InputRange=(X=0,Y=-100), TwistAxis="1,0,0")
   ) -> (Output_Get:double)
   > Set Alpha_ThighR_Front(Alpha_ThighR_Front:double=GetAlphaFromReferenceTransformDelta(SkeletalMeshComponent=@n5, BoneName=thigh_r, BaseName=pelvis, InputRange=(X=0,Y=-100), TwistAxis="1,0,0")) -> (Output_Get:double)
   > Set Alpha_Calf_L(Alpha_Calf_L:double=GetAlphaFromReferenceTransformDelta(SkeletalMeshComponent=@n5, BoneName=calf_l, BaseName=thigh_l, InputRange=(X=40,Y=80), TwistAxis="1,0,0")) -> (Output_Get:double)
   > Set Alpha_Calf_R(Alpha_Calf_R:double=GetAlphaFromReferenceTransformDelta(SkeletalMeshComponent=@n5, BoneName=calf_r, BaseName=thigh_r, InputRange=(X=40,Y=85), TwistAxis="1,0,0")) -> (Output_Get:double)
   n6 = Gravity Override State Function DEPRECATED(
    execute=@n4.then_2
    CharacterPartAnimInstance=Self-Reference()
    GameTimeInSeconds=GetGameTimeInSeconds()
    GravityOverrideParams=Get Gravity Override Params()
   )
   > Set Gravity Override(Gravity Override=@n6)
   n7 = Get bIsMoving2D()
   > Set Crouch stop(Crouch stop:bool=AND Boolean(A=Get bIsCrouching(), B=NOT Boolean(A=@n7))) -> (Output_Get:bool)
   n8 = Get bIsSkyDiving()
   n9 = Get bIsJumping()
   n10 = Get bIsFalling()
   > Set Simple jump(Simple jump:bool=AND Boolean(A=OR Boolean(A=@n9, B=@n10), B=NOT Boolean(A=@n8))) -> (Output_Get:bool)
   Is Valid(exec=@n4.then_0, InputObject=Get PlayerPawn() -> (PlayerPawn:Fort Player Pawn Athena Object Reference))
   > Set WPO Globals Function(
    Material=Get Mat Instance Dynamic()
    Overlay Material=Get Overlay Mat Instance Dynamic()
    Skeletal Mesh=Get Skel Mesh()
    "Local Noise Transform  Location"="Get Local Noise Transform  Location"() -> ("Local Noise Transform  Location":Vector)
    Local Noise Offset=Get Local Noise Offset()
   )
   > Set WPO Regional Function!(
    execute
    Material:Material Instance Dynamic Object Reference=Get Mat Instance Dynamic()
    Overlay Material:Material Instance Dynamic Object Reference=Get Overlay Mat Instance Dynamic()
    Skeletal Mesh:Skeletal Mesh Component Object Reference=Get Skel Mesh()
    Velocity Transform:Name=foot_r
    Velocity Interp Speed:double=1
    Collision Transform:Name=calf_r
    Collision Transform Axis:EAxis Enum
    Invert Collision Transform Axis:bool=false
    Delta Time X:double=@n3.DeltaTimeX
    Previous Velocity Transform Location:"Vector (by ref)"=Get R1 Velocity Transform Location()
    Previous Velocity Transform Velocity:"Vector (by ref)"=Get R1 Velocity Transform Velocity()
    Region:int=1
   ) -> (then)
   > Set WPO Regional Function(
    Material=Get Mat Instance Dynamic()
    Overlay Material=Get Overlay Mat Instance Dynamic()
    Skeletal Mesh=Get Skel Mesh()
    Velocity Transform=foot_l
    Velocity Interp Speed=1
    Collision Transform=calf_l
    Invert Collision Transform Axis=true
    Delta Time X=@n3.DeltaTimeX
    Previous Velocity Transform Location=Get R2 Velocity Transform Location()
    Previous Velocity Transform Velocity=Get R2 Velocity Transform Velocity()
    Region=1
   )
   > n11 = Branch(Condition=Get bIsSkydiveFloating())
   > n12 = Set WPO Pho Wind Intensity(WPO Pho Wind Intensity=1)
   SetScalarParameterValue(execute=[@n13.then, @n12.then], self=Get Mat Instance Dynamic(), ParameterName=ClothSim_WindStrength, Value=Get WPO Pho Wind Intensity())
   n13 = Set WPO Pho Wind Intensity(execute=@n11.else, WPO Pho Wind Intensity=0)
   Event OnCharacterCustomizationLoaded()
   > n14 = Set Skel Mesh(Skel Mesh=GetPartSkeletalMeshComponent())
   > n15 = Get Dynamic Material Instance(SkeletalMeshComponent=@n14)
   > Set Mat Instance Dynamic(Mat Instance Dynamic=@n15)
   > n16 = GetOrCreateDynamicOverlayMaterial(MeshComponent=@n14, MaterialName=None)
   > Set Overlay Mat Instance Dynamic(Overlay Mat Instance Dynamic=@n16)
  ]}
]
variables [
 Alpha_Calf_L {type=number}
 Alpha_Calf_R {type=number}
 Alpha_ThighL_Front {type=number}
 Alpha_ThighR_Front {type=number}
 Crouch stop {type=boolean}
 Gravity Override {type=Vector}
 Gravity Override Params {type=GravityOverrideParamsStruct}
 Local Noise Offset {type=Vector}
 "Local Noise Transform  Location" {type=Vector}
 Mat Instance Dynamic {type=MaterialInstanceDynamic}
 Overlay Mat Instance Dynamic {type=MaterialInstanceDynamic}
 PlayerPawn {type=FortPlayerPawnAthena}
 R1 Velocity Transform Location {type=Vector}
 R1 Velocity Transform Velocity {type=Vector}
 R2 Velocity Transform Location {type=Vector}
 R2 Velocity Transform Velocity {type=Vector}
 Simple jump {type=boolean}
 Skel Mesh {type=SkeletalMeshComponent}
 WPO Pho Wind Intensity {type=number}
]
